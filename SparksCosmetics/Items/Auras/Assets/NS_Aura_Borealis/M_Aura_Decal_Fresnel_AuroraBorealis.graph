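# Translucent unlit Niagara sprite material for the Borealis aura, faking a decal glow with aurora-like noise.
# Squared particle color drives emissive, while two panning noise samples, one with symmetrical UVs, multiply into opacity with particle alpha.
# A sphere mask and inverted depth fade confine it near scene geometry.
# Fresnel from flattened SceneTexture world normals shapes opacity, replaced by a constant parameter on mobile and low quality.
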Material /SparksCosmetics/Items/Auras/Assets/NS_Aura_Borealis/M_Aura_Decal_Fresnel_AuroraBorealis
graph {
 properties {bUsedWithNiagaraSprites=true, blendMode=BLEND_Translucent, pixelDepthOffsetMode=PDOM_Legacy, shadingModel=MSM_Unlit, usedShadingModels=Unlit}
 nodes [
  n1 = ParticleColor()
  MaterialFunctionCall(Emissive=Multiply(A=ScalarParameter() {defaultValue=5, parameterName=EmissiveMult}, B=Multiply(A=@n1.RGB, B=@n1.RGB))) {materialFunction=&MF_FX_EmissiveAdjustments}
  n2 = TextureCoordinate()
  n3 = Multiply(
   A=MaterialFunctionCall!(In UV=@n2) {materialFunction=/Game/Effects/Niagara/MasterMaterials/Functions/Mojo_MF/MF_FX_Symmetrical_UV}.Symmetric UV Vert
   B=AppendVector(A=ScalarParameter() {defaultValue=3, parameterName=Mask1 Tile X}, B=Constant() {r=0.1})
  )
  n4 = AppendVector(A=Constant(), B=Constant())
  n5 = Constant() {r=7}
  n6 = Multiply(A=Time(), B=Constant() {r=0.5})
  n7 = ScalarParameter() {defaultValue=0.05, parameterName=Rand Opacity Speed X}
  n8 = ScalarParameter() {defaultValue=0.05, parameterName=Rand Opacity Speed Y}
  n9 = TextureObject() {samplerType=SAMPLERTYPE_LinearColor, texture=&HighContrastNoise}
  n10 = TextureSample(UVs=Panner(Coordinate=Add(A=@n3, B=@n4), Time=Multiply(A=@n6, B=@n5), Speed=AppendVector(A=@n7, B=@n8)), Tex=@n9)
  n11 = Multiply(A=@n2, B=AppendVector(A=ScalarParameter() {defaultValue=4, parameterName=Mask2 Tile X}, B=Constant() {r=0.1}))
  n12 = AppendVector(A=Constant() {r=0.5432}, B=Constant() {r=0.2543})
  n13 = Constant() {r=8}
  n14 = ScalarParameter() {defaultValue=0.05, parameterName=Rand Opacity Speed X}
  n15 = ScalarParameter() {defaultValue=0.05, parameterName=Rand Opacity Speed Y}
  n16 = TextureSample(UVs=Panner(Coordinate=Add(A=@n11, B=@n12), Time=Multiply(A=@n6, B=@n13), Speed=AppendVector(A=@n14, B=@n15)), Tex=@n9) {samplerType=SAMPLERTYPE_LinearColor, texture=&HighContrastNoise}
  n17 = ScalarParameter() {defaultValue=1.5, parameterName=Random Opacity Pow}
  n18 = Multiply(A=Multiply(A=Multiply(A=@n10.G), B=Power(Base=@n16.R, Exp=@n17)), B=ScalarParameter() {defaultValue=5, parameterName=Noise Alpha Mult})
  n19 = DynamicParameter() {defaultValue=<rgba 3 0.15 1 1>, paramNames [FresnelExponent, FlattenNormal, Param3, Param4]}
  n20 = SceneTexture() {sceneTextureId=PPI_WorldNormal}
  n21 = Fresnel(
   ExponentIn=@n19.FresnelExponent
   Normal=MaterialFunctionCall(Normal=ComponentMask(None=@n20.Color) {b=true, g=true, r=true}, Flatness=@n19.FlattenNormal) {materialFunction=&FlattenNormal}.Result
  ) {baseReflectFraction=0, exponent=1}
  n22 = VectorParameter() {parameterName=Forward, defaultValue=<rgba 0 0 1 1>}
  n23 = ShadingPathSwitch(Default=QualitySwitch(Default=@n21, Low=@n22.RGB, High=@n21, Medium=@n21, Epic=@n21), Mobile=@n22.RGB)
  n24 = SphereMask(
   A=TextureCoordinate()
   B=Constant() {r=0.5}
   Radius=ScalarParameter() {defaultValue=0.5, parameterName=SphereMaskRadius}
   Hardness=ScalarParameter() {parameterName=SphereMaskHardness}
  ) {hardnessPercent=0}
  n25 = OneMinus(None=DepthFade(FadeDistance=ScalarParameter() {defaultValue=15, parameterName=VisibleDepthDistance}) {fadeDistanceDefault=80})
  n26 = ScalarParameter() {defaultValue=15, parameterName=DepthFade}
  Multiply(
   A=Multiply(A=Saturate(None=@n18) {desc=opacity}, B=@n1.A)
   B=Multiply(A=Multiply(A=@n23, B=@n24), B=DepthFade(Opacity=@n25, FadeDistance=@n26) {fadeDistanceDefault=15})
  )
  TextureCoordinate()
 ]
}
