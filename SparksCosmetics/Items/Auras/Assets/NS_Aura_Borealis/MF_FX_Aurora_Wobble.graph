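# Material function that outputs a wobbling World Position Offset for the Borealis aurora aura effect.
# Per-axis sine waves use world position, time, speed, offset and scale, with amplitude multiplied by clamped wind-vector length.
# Each axis then sums the other two waves, scaled by strength and red vertex color.
# A stepped Normal Direction input selects whether the offset is multiplied by the vertex normal; a Wind function call sits unconnected.
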MaterialFunction /SparksCosmetics/Items/Auras/Assets/NS_Aura_Borealis/MF_FX_Aurora_Wobble
graph {nodes [
  n1 = Constant() {desc=Wobble Speed, r=1}
  n2 = Add(
   A=Multiply(A=Time(), B=FunctionInput(Preview=@n1) {bUsePreviewValueAsDefault=true, inputName=Wobble Speed, inputType=FunctionInput_Scalar})
   B=FunctionInput(Preview=Constant() {r=1}) {bUsePreviewValueAsDefault=true, inputName=Wobble Offset, inputType=FunctionInput_Scalar, sortPriority=1}
  )
  n3 = FunctionInput(Preview=WorldPosition() {worldPositionShaderOffset=WPT_ExcludeAllShaderOffsets}.XYZ) {bUsePreviewValueAsDefault=true, inputName=World Position, sortPriority=8}
  n4 = ComponentMask(None=@n3) {r=true}
  n5 = Multiply(A=FunctionInput(Preview=Constant() {r=1}) {bUsePreviewValueAsDefault=true, inputName=Wobble Scale, inputType=FunctionInput_Scalar, sortPriority=2}) {constB=0.1, desc=Adjust Scale}
  n6 = Multiply(A=Constant() {r=-0.3854}, B=Add(A=@n2, B=Divide(A=@n4, B=@n5)))
  n7 = ComponentMask(None=@n3) {g=true}
  n8 = Multiply(A=Constant() {r=0.5}, B=Add(A=@n2, B=Divide(A=@n7, B=@n5)))
  n9 = ComponentMask(None=@n3) {b=true}
  n10 = Multiply(A=Constant() {r=0.4452}, B=Add(A=@n2, B=Divide(A=@n9, B=@n5)))
  n11 = FunctionInput(Preview=Constant() {r=0.1}) {bUsePreviewValueAsDefault=true, desc=0.1-1, inputName=Wobble Amplitude, inputType=FunctionInput_Scalar, sortPriority=3}
  n12 = MaterialFunctionCall(Vector 3=FunctionInput(Preview=Constant3Vector()) {inputName=Normalized Wind Vector, sortPriority=10}) {materialFunction=/Engine/Functions/Engine_MaterialFunctions02/Utility/VectorLength}
  n13 = Multiply(
   A=MaterialFunctionCall(X=Sine(None=@n6), Y=Sine(None=@n8), Z=Sine(None=@n10)) {materialFunction=&MakeFloat3}.Result
   B=Multiply(A=@n11, B=Clamp(None=@n12.V3 Length)) {desc=Amplitude}
  )
  n14 = ComponentMask(None=@n13) {g=true}
  n15 = ComponentMask(None=@n13) {b=true}
  n16 = Add(A=@n14, B=@n15)
  n17 = ComponentMask(None=@n13) {r=true}
  n18 = Add(A=@n17, B=@n15)
  n19 = Add(A=@n17, B=@n14)
  n20 = FunctionInput(Preview=Constant3Vector() {constant=<rgba 0.1 0.1 0.1 0>}) {bUsePreviewValueAsDefault=true, desc=0.1-1, inputName=Wobble Strength, sortPriority=4}
  n21 = Constant() {r=0.1}
  n22 = Multiply(
   A=Multiply(
    A=MaterialFunctionCall(X=@n16, Y=@n18, Z=@n19) {materialFunction=&MakeFloat3}.Result
    B=Multiply(A=@n20, B=@n21) {desc=Adjust Strength}
   )
   B=VertexColor().R
  )
  n23 = Constant() {desc=Vertex Color Alpha Interpolator}
  n24 = LinearInterpolate(
   A=@n22
   B=Multiply(A=@n22, B=VertexNormalWS())
   Alpha=Step(Y=FunctionInput(Preview=@n23) {bUsePreviewValueAsDefault=true, inputName=Normal Direction?, inputType=FunctionInput_Scalar, sortPriority=9}) {constX=0.5}
  )
  Add(A=@n24, B=@n3)
  FunctionOutput(None=@n24) {description=World Space Position Modifier, outputName=Out WPO}
  MaterialFunctionCall!(WindActor) {materialFunction=/Engine/Functions/Engine_MaterialFunctions02/WorldPositionOffset/Wind}
 ]}
