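# Unlit, translucent, two-sided base material for Niagara mesh particles in the Aurora Borealis aura cosmetic.
# Skewed, mirrored UVs drive two panning noise samples that shape curtain opacity and distort a tinted color gradient.
# Emissive scales with time of day, while opacity combines particle alpha, facing-angle falloff, camera-depth fade and DepthFade.
# A wobble function, randomized per particle, supplies world position offset masked by vertex alpha.
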Material /SparksCosmetics/Items/Auras/Assets/NS_Aura_Borealis/M_AuroraBorealis_Base
graph {
 properties {
  bUsedWithNiagaraMeshParticles=true
  blendMode=BLEND_Translucent
  pixelDepthOffsetMode=PDOM_Legacy
  shadingModel=MSM_Unlit
  twoSided=true
  usedShadingModels=Unlit
 }
 nodes [
  n1 = MaterialFunctionCall!(
   Wobble Speed=ScalarParameter() {defaultValue=0.3, parameterName=Wobble Speed}
   Wobble Offset=Add(A=ScalarParameter() {defaultValue=1, parameterName=Wobble Offset}, B=ParticleRandom())
   Wobble Scale=ScalarParameter() {defaultValue=50, parameterName=WPO Scale}
   Wobble Amplitude=Constant() {r=0.75}
   Wobble Strength=ScalarParameter() {defaultValue=50, parameterName=WPO Intensity}
   World Position=Constant3Vector()
   Normal Direction?=Constant()
   Normalized Wind Vector=Constant3Vector() {desc=temp wind direction, constant=<rgba 1 0 0 0>}
  ) {materialFunction=&/MF_FX_Aurora_Wobble}
  n2 = ParticlePositionWS()
  Add(A=@n1.Out WPO, B=@n2)
  Constant() {desc=debug alpha, r=1}
  Constant() {desc=large, r=1000}
  Constant() {desc=large, r=1000}
  n3 = Fresnel(
   ExponentIn=ScalarParameter() {defaultValue=1, parameterName=ExponentIn}
   BaseReflectFractionIn=ScalarParameter() {parameterName=BaseReflectFractionIn}
  ) {exponent=2}
  LinearInterpolate(A=Constant() {r=1}, B=Constant(), Alpha=@n3)
  n4 = Constant() {r=1}
  n5 = Constant() {r=2}
  n6 = ScalarParameter() {defaultValue=3, parameterName=Emission Mult}
  n7 = TextureCoordinate()
  n8 = ComponentMask(None=@n7) {g=true}
  n9 = ScalarParameter() {defaultValue=0.045, parameterName=Skew Amount}
  n10 = Constant() {r=-0.5}
  n11 = Multiply(A=Add(A=Multiply(A=@n8, B=@n9), B=@n9), B=@n10)
  n12 = OneMinus(None=@n8)
  n13 = Multiply(A=Add(A=Multiply(A=@n12, B=@n9), B=@n9), B=@n10)
  n14 = Add(A=@n7, B=AppendVector(A=Subtract(A=@n11, B=@n13), B=Constant()))
  n15 = Multiply(
   A=MaterialFunctionCall!(In UV=@n14) {materialFunction=/Game/Effects/Niagara/MasterMaterials/Functions/Mojo_MF/MF_FX_Symmetrical_UV}.Symmetric UV Vert
   B=AppendVector(A=ScalarParameter() {defaultValue=5, parameterName=Mask1 Tile X}, B=Constant() {r=0.1})
  )
  n16 = AppendVector(A=Constant(), B=Constant())
  n17 = Constant() {r=7}
  n18 = Multiply(A=Time(), B=Constant() {r=0.5})
  n19 = ScalarParameter() {defaultValue=0.05, parameterName=Rand Opacity Speed X}
  n20 = ScalarParameter() {defaultValue=0.05, parameterName=Rand Opacity Speed Y}
  n21 = TextureObject() {samplerType=SAMPLERTYPE_LinearColor, texture=&HighContrastNoise}
  n22 = TextureSample(UVs=Panner(Coordinate=Add(A=@n15, B=@n16), Time=Multiply(A=@n18, B=@n17), Speed=AppendVector(A=@n19, B=@n20)), Tex=@n21)
  n23 = ScalarParameter() {defaultValue=6, parameterName=Mask2 Tile X}
  n24 = Constant() {r=0.1}
  n25 = Add(A=Multiply(A=@n14, B=AppendVector(A=@n23, B=@n24)), B=AppendVector(A=Constant() {r=0.5432}, B=Constant() {r=0.2543}))
  n26 = Multiply(A=@n18, B=Constant() {r=8})
  n27 = AppendVector(
   A=ScalarParameter() {defaultValue=0.05, parameterName=Rand Opacity Speed X}
   B=ScalarParameter() {defaultValue=0.05, parameterName=Rand Opacity Speed Y}
  )
  n28 = Power(
   Base=TextureSample(UVs=Panner(Coordinate=@n25, Time=@n26, Speed=@n27), Tex=@n21) {samplerType=SAMPLERTYPE_LinearColor, texture=&HighContrastNoise}.R
   Exp=ScalarParameter() {defaultValue=1.5, parameterName=Random Opacity Pow}
  )
  n29 = Multiply(A=Multiply(A=@n22.G), B=@n28)
  n30 = Constant() {r=1}
  n31 = ComponentMask(None=Multiply() {constB=0.1}) {g=true}
  n32 = Multiply(A=Power(Base=OneMinus(None=@n31)) {constExponent=100}) {constB=0.45}
  n33 = TextureCoordinate() {uTiling=0.5, vTiling=0.5}
  n34 = MaterialFunctionCall(Float2=Sine(None=Frac(None=@n33))) {materialFunction=&BreakOutFloat2Components}
  n35 = Multiply(A=Multiply(A=@n34.R, B=@n34.R))
  n36 = Multiply(A=Power(Base=Abs(None=@n35)) {constExponent=3})
  n37 = Multiply(A=@n36, B=Add(A=@n32, B=Multiply(A=@n30, B=@n29)))
  n38 = Multiply(A=@n22.B, B=Power(Base=@n28, Exp=ScalarParameter() {defaultValue=0.3, parameterName=Random Opacity Secondary Pow}))
  n39 = Multiply(A=@n36, B=Multiply(A=@n30, B=@n38))
  n40 = ComponentMask(None=CameraVectorWS()) {g=true, r=true}
  n41 = Multiply(A=@n18, B=Constant() {r=0.005})
  n42 = Add(A=Add(A=Add(A=@n40, B=@n41), B=@n38), B=ParticleRandom())
  n43 = TextureObject() {texture=&/T_Aurora_Grad_Blur_Small_D}
  n44 = TextureCoordinate()
  n45 = Saturate(None=Add(A=ComponentMask(None=@n44) {g=true}, B=Multiply(A=@n38) {constB=0.05}))
  n46 = LinearInterpolate(
   A=LinearInterpolate(
    A=VectorParameter() {parameterName=TopColor, defaultValue=<rgba 0.9896 0 0.8503 1>}.RGB
    B=VectorParameter() {parameterName=BottomColor, defaultValue=<rgba 0.2715 0.9844 0.5051 1>}.RGB
    Alpha=@n45
   )
   B=VectorParameter() {parameterName=EdgeColor, defaultValue=<rgba 0.1365 0.7 0.602 1>}.RGB
   Alpha=Power(Base=@n45)
  )
  n47 = Multiply(A=LinearInterpolate(A=TextureSample(UVs=@n42, Tex=@n43).RGB, B=Multiply(A=@n46)) {constAlpha=0.8}, B=Multiply(A=@n32, B=@n39))
  n48 = Constant() {r=-5}
  n49 = LinearInterpolate(
   A=Multiply(A=Desaturation(None=@n47, Fraction=@n48)) {constB=2, desc=Optional Boost}
   B=Multiply(A=VectorParameter() {parameterName=Primary Color, defaultValue=<rgba 0.04 0.973 0.0029 1>}.RGB, B=@n37)
   Alpha=Constant() {r=0.25}
  )
  MaterialFunctionCall(Emissive=Multiply(
    A=MaterialFunctionCall(Value_Morning=@n4, Value_Day=@n5, Value_Evening=@n4, Value_Night=@n5) {materialFunction=&Scalar_Time_of_Day}.Result
    B=Multiply(A=@n6, B=@n49)
   )) {materialFunction=&MF_FX_EmissiveAdjustments}
  n50 = MaterialFunctionCall(Float3=@n1.Out WPO) {materialFunction=&BreakOutFloat3Components}
  MaterialFunctionCall(
   X=@n50.R
   Y=@n50.R
   Z=MaterialFunctionCall(Float3=@n2) {materialFunction=&BreakOutFloat3Components}.B
  ) {materialFunction=&MakeFloat3}
  n51 = Multiply(A=Max(A=@n37, B=@n39), B=ScalarParameter() {defaultValue=1, parameterName=Overall Alpha Mult})
  Multiply(A=OneMinus(None=@n3), B=@n51)
  Multiply(A=@n3) {constB=1.25}
  Multiply(A=TextureCoordinate())
  n52 = VertexColor()
  Multiply(A=@n52)
  OneMinus(None=@n52.A)
  ParticlePositionWS()
  Reroute(None=Multiply(A=@n1.Out WPO, B=@n52.A)) {desc=WPO}
  n53 = TextureCoordinate()
  n54 = Constant2Vector() {g=0.5, r=0.5}
  n55 = Constant() {r=0.8}
  n56 = Constant() {r=0.001}
  Reroute(None=Power(Base=SphereMask(A=@n53, B=@n54, Radius=@n55, Hardness=@n56), Exp=Constant() {r=6}))
  n57 = CameraVectorWS()
  n58 = VertexNormalWS()
  n59 = Power(Base=Abs(None=DotProduct(A=@n57, B=@n58)) {desc=for 2 sided}, Exp=Constant() {desc=edge sharpness, r=3})
  n60 = Saturate(None=@n51) {desc=opacity}
  n61 = Multiply(A=ParticleColor().A, B=Multiply(A=@n60, B=Saturate(None=@n59)) {desc=texture alpha})
  MaterialExpressionSubtract_26 = Subtract(A=CameraPositionWS(), B=ParticlePositionWS())
  n62 = Custom!(x=@MaterialExpressionSubtract_26) {
   code="length(x)"
   inputs [{inputName=x, input {expression=&self:MaterialExpressionSubtract_26}}]
  }
  n63 = Constant3Vector() {constant=<rgba 0 0 -1 0>}
  n64 = Multiply(
   A=@n62
   B=DotProduct(A=Normalize(VectorInput=@MaterialExpressionSubtract_26), B=Transform(None=@n63) {transformSourceType=TRANSFORMSOURCE_View})
  )
  n65 = Constant() {desc=Fade Offset, r=24}
  n66 = Clamp(None=Divide(A=Subtract(A=@n64, B=@n65) {constB=0}, B=Constant() {desc=Fade Length, r=512}))
  n67 = Saturate(None=Multiply(A=OneMinus(None=@n66)))
  n68 = Constant()
  n69 = Constant() {r=0.35}
  n70 = Constant() {r=1}
  n71 = Saturate(None=Multiply(A=MaterialFunctionCall(Input=@n67, Input Low=@n68, Input High=@n70, Target Low=@n69, Target High=@n70) {materialFunction=&RemapValueRange}.Result) {constB=2})
  Reroute(None=Multiply(A=Multiply(A=@n61, B=@n71), B=DepthFade())) {desc=opacity angle fade}
  n72 = Add(A=Divide(A=WorldPosition().XYZ, B=Constant() {r=500}), B=Constant())
  n73 = MaterialFunctionCall(X=ComponentMask(None=@n72) {g=true}, Y=ComponentMask(None=@n72) {b=true}) {materialFunction=&MakeFloat2}
  n74 = TextureSample(
   UVs=Panner(Coordinate=Multiply(A=@n73.Result) {constB=0.1}, Time=Time(), Speed=Constant2Vector() {g=0.1, r=0.01})
   Tex=TextureObjectParameter() {parameterName=Noise Mask Texture, texture=/Game/Global/Volumetric/T_Cloud2dNoise}
  )
  Reroute(None=Multiply(A=ConstantBiasScale(None=@n74.R), B=Constant() {r=1}))
  Reroute(None=@n60) {desc=opacity}
  Saturate(None=@n3)
  TextureCoordinate()
  VertexColor()
  WorldPosition()
  WorldPosition() {worldPositionShaderOffset=WPT_ExcludeAllShaderOffsets}
  WorldPosition()
 ]
}
