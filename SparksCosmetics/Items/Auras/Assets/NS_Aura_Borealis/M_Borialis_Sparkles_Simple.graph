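# Translucent unlit Niagara sprite material for the audio-reactive sparkles of the Borealis aura in Sparks cosmetics.
# Particle color feeds an emissive adjustment function, while particle alpha is multiplied by the green channel of a four-point star texture.
# That mask is scaled by a saturated sum of MeasureLifetime and BeatLifetime from the Sparks audio parameter collection, so sparkles pulse with the music.
# A Beat_Intensity parameter weights the beat term.
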Material /SparksCosmetics/Items/Auras/Assets/NS_Aura_Borealis/M_Borialis_Sparkles_Simple
graph {
 properties {
  bEnableResponsiveAA=true
  bUsedWithNiagaraSprites=true
  blendMode=BLEND_Translucent
  pixelDepthOffsetMode=PDOM_Legacy
  shadingModel=MSM_Unlit
  usedShadingModels=Unlit
 }
 nodes [
  n1 = Add(
   A=CollectionParameter() {collection=/FMAudioReactiveData/AudioAnalysis/MPC_Sparks_Audio_Parameters, parameterName=MeasureLifetime}
   B=Multiply(
    A=CollectionParameter() {collection=&MPC_Sparks_Audio_Parameters, parameterName=BeatLifetime}
    B=ScalarParameter() {defaultValue=1, parameterName=Beat_Intensity}
   )
  )
  Clamp(None=@n1) {maxDefault=2500, minDefault=0.5}
  MaterialFunctionCall(Emissive=ParticleColor().RGB) {materialFunction=&MF_FX_EmissiveAdjustments}
  Multiply(
   A=Multiply(A=ParticleColor().A, B=TextureSample() {texture=&/T_Borialis_Sparkles_4Star_01_D}.G)
   B=Saturate(None=@n1)
  ) {constB=150}
 ]
}
